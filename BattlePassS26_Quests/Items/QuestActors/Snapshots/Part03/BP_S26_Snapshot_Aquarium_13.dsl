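; Season 26 Battle Pass snapshot quest actor (Aquarium 13) that stages a gym-themed prop reveal when its quest becomes ready.
; BeginPlay applies a plushie free-weight material and initial custom primitive data to the pre-interaction mesh.
; OnQuestReady notifies kettlebell rack, tire pile, buoy and jump rope child actors through casts.
; It then hides the mesh, reshows it after a short delay and drives its float-down material animation through timeline-fed custom primitive data.
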(event EventBeginPlay
  (bind _sm_quest_preinteraction (Variables|Components|GetSM_Quest_PreInteraction))
  (Rendering|Material|SetMaterial _sm_quest_preinteraction 0 "/BattlePassS26_Quests/Items/QuestActors/Snapshots/Part03/Material/MI_Quest_PlushieFreeWeight_A.MI_Quest_PlushieFreeWeight_A")
  (Rendering|Material|SetDefaultCustomPrimitiveDataFloat _sm_quest_preinteraction 1 1.0)
  (Rendering|Material|SetDefaultCustomPrimitiveDataFloat _sm_quest_preinteraction 2 1.0))

(event Collision|EventActorBeginOverlap
  (params (OtherActor :type Actor%20Object%20Reference)))

(event EventTick
  (params (DeltaSeconds :type Float%20%28single-precision%29)))

(event OnQuestReady(ConsolidatedQuestComponent)
  (params (PlayerController :type Fort%20Player%20Controller%20Object%20Reference))
  (bind _sm_quest_preinteraction (Variables|Components|GetSM_Quest_PreInteraction))
  (bind _sm_quest_preinteraction_2 (Variables|Components|GetSM_Quest_PreInteraction))
  (Utilities|FlowControl|Sequence
    (:then_0
      (bind _asbp_gym_kettle_bell_rack_b_quest (Utilities|Casting|CastToBP_Gym_KettleBellRack_B_quest (Variables|/Script/Engine.ChildActorComponent|GetChildActor (Variables|Default|GetBP_Gym_KettleBellRack_B))))
      (Default|CallGYMKettleBellQuestReady _asbp_gym_kettle_bell_rack_b_quest))
    (:then_1
      (bind _asbp_garage_tires_pile_a_quest (Utilities|Casting|CastToBP_Garage_TiresPile_A_Quest (Variables|/Script/Engine.ChildActorComponent|GetChildActor (Variables|Default|GetBP_Garage_TiresPile_A))))
      (Default|CallGYMGarageQuestReady _asbp_garage_tires_pile_a_quest))
    (:then_2
      (bind _asbp_coastal_buoy_a_quest (Utilities|Casting|CastToBP_Coastal_Buoy_A_quest (Variables|/Script/Engine.ChildActorComponent|GetChildActor (Variables|Default|GetBP_Coastal_Buoy_A))))
      (Default|CallGYMCoastalQuestready _asbp_coastal_buoy_a_quest))
    (:then_3
      (bind _asbp_gym_jump_rope_a_quest (Utilities|Casting|CastToBP_Gym_JumpRope_A_quest (Variables|/Script/Engine.ChildActorComponent|GetChildActor (Variables|Default|GetBP_Gym_JumpRope_A))))
      (Default|CallGYMJumpRopeQuestReady _asbp_gym_jump_rope_a_quest))
    (:then_4
      (Rendering|SetVisibility _sm_quest_preinteraction false false)
      (Utilities|FlowControl|Delay 0.8)
      (Rendering|SetVisibility _sm_quest_preinteraction true false)
      (bind (_direction _float_down_process _float_down_animation) (Timeline|Timeline_2 Play)
        (:Update
          (Rendering|Material|SetDefaultCustomPrimitiveDataFloat _sm_quest_preinteraction_2 1 _float_down_process)
          (Rendering|Material|SetDefaultCustomPrimitiveDataFloat _sm_quest_preinteraction_2 2 _float_down_animation))))))

(fn ConstructionScript)
