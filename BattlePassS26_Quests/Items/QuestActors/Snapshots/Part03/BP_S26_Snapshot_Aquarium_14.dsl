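; Season 26 Battle Pass quest actor for a Snapshot quest, apparently number 14 in an Aquarium set, going by its name and path.
; Its event graph holds only empty stubs for BeginPlay, actor overlap, Tick and the ConsolidatedQuestComponent's OnQuestReady event.
; The construction script is also empty, so any behavior comes from a parent class, components or defaults not shown here.
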(event EventBeginPlay)

(event Collision|EventActorBeginOverlap
  (params (OtherActor :type Actor%20Object%20Reference)))

(event EventTick
  (params (DeltaSeconds :type Float%20%28single-precision%29)))

(event OnQuestReady(ConsolidatedQuestComponent)
  (params (PlayerController :type Fort%20Player%20Controller%20Object%20Reference)))

(fn ConstructionScript)
