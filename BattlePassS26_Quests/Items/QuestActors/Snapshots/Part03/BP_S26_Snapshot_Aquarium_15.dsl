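; Blueprint for an Aquarium snapshot quest actor, judging by its name, in the Battle Pass Season 26 quests.
; On BeginPlay it fetches the AmbientAudio component and, on the success branch, registers it with the MusicEventSubsystem under the diegetic POI event tag.
; The construction script is empty, and the file shows no other logic.
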(event EventBeginPlay
  (bind _ambientaudio (Variables|Components|GetAmbientAudio)
    (:then
      (Audio|AddEventbyComponent (GameInstanceSubsystems|GetMusicEventSubsystem) _ambientaudio "(TagName=\"MusicEventSubsystem.Event.POI.Diegetic\")" "(GameplayTags=)" 0 ""))
    (:else)))

(fn ConstructionScript)
