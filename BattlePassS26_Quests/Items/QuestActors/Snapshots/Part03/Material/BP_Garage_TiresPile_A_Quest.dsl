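; Tire pile quest actor for the GYM garage that plays a material-driven float-down animation when the quest becomes ready.
; BeginPlay sets custom primitive data slots 1 and 2 to full on the static mesh and binds to the quest-ready event.
; That event hides the mesh, shows it after a short delay, then runs a timeline feeding process and animation values into both slots.
; Overlap and Tick are empty.
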(event Collision|EventActorBeginOverlap
  (params (OtherActor :type Actor%20Object%20Reference)))

(event EventBeginPlay
  (bind _staticmeshcomponent (Variables|Debug|GetStaticMeshComponent))
  (Rendering|Material|SetDefaultCustomPrimitiveDataFloat _staticmeshcomponent 1 1.0)
  (Rendering|Material|SetDefaultCustomPrimitiveDataFloat _staticmeshcomponent 2 1.0)
  (Default|BindEventtoGYMGarageQuestReady (delegate Custom|GYM_garage_questready)))

(event EventTick
  (params (DeltaSeconds :type Float%20%28single-precision%29)))

(event Custom|GYM_garage_questready
  (bind _staticmeshcomponent (Variables|Debug|GetStaticMeshComponent))
  (bind _staticmeshcomponent_2 (Variables|Debug|GetStaticMeshComponent))
  (Rendering|SetVisibility (Variables|Debug|GetStaticMeshComponent) false false)
  (Utilities|FlowControl|Delay 0.5)
  (Rendering|SetVisibility (Variables|Debug|GetStaticMeshComponent) true false)
  (bind (_direction _float_down_process _float_down_animation) (Timeline|Timeline Play)
    (:Update
      (Rendering|Material|SetDefaultCustomPrimitiveDataFloat _staticmeshcomponent 1 _float_down_process)
      (Rendering|Material|SetDefaultCustomPrimitiveDataFloat _staticmeshcomponent 2 _float_down_animation))
    (:Finished
      (Rendering|Material|SetDefaultCustomPrimitiveDataFloat _staticmeshcomponent_2 1 1.0)
      (Rendering|Material|SetDefaultCustomPrimitiveDataFloat _staticmeshcomponent_2 2 1.0))))

(fn ConstructionScript
  (|CallParentFunction))
