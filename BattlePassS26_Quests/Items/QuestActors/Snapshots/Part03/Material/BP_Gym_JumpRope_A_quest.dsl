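; Battle Pass S26 quest Blueprint for a gym jump rope prop that animates its mesh material when the jump rope quest signals ready.
; BeginPlay binds a custom event to the GYMJumpRopeQuestReady dispatcher, and that event plays a timeline.
; Each timeline update writes float-down process and animation values into default custom primitive data slots 1 and 2 on the static mesh component, so the material receives them without a dynamic instance.
; Tick and overlap events are empty.
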(event EventBeginPlay
  (Default|BindEventtoGYMJumpRopeQuestReady (delegate Custom|GYMJumpRopeQuest)))

(event Collision|EventActorBeginOverlap
  (params (OtherActor :type Actor%20Object%20Reference)))

(event Custom|GYMJumpRopeQuest
  (bind _staticmeshcomponent (Variables|Debug|GetStaticMeshComponent))
  (bind (_direction _float_down_process _float_down_animation) (Timeline|Timeline Play)
    (:Update
      (Rendering|Material|SetDefaultCustomPrimitiveDataFloat _staticmeshcomponent 1 _float_down_process)
      (Rendering|Material|SetDefaultCustomPrimitiveDataFloat _staticmeshcomponent 2 _float_down_animation))))

(event EventTick
  (params (DeltaSeconds :type Float%20%28single-precision%29)))

(fn ConstructionScript
  (|CallParentFunction))
