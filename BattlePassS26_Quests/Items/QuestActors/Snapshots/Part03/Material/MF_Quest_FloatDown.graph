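# Material function for quest actors that produces a world position offset making the mesh float down into place with a gentle wobble.
# A custom primitive data progress value blends a vertical offset from raised height to zero.
# Time-driven sine rotations around X, Y and Z, pivoting on the object pivot plus an adjustable offset, are summed with it.
# A second custom primitive data value fades the wobble out through a named reroute.
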MaterialFunction /BattlePassS26_Quests/Items/QuestActors/Snapshots/Part03/Material/MF_Quest_FloatDown
graph {nodes [
  ActorPositionWS()
  ActorPositionWS()
  ActorPositionWS()
  n1 = ScalarParameter() {defaultValue=360, group=Quest float down, parameterName=Vertical motion}
  n2 = ScalarParameter() {bUseCustomPrimitiveData=true, defaultValue=1, group=Quest float down, parameterName=Float down process, primitiveDataIndex=1, sliderMax=1}
  n3 = Saturate(None=@n2)
  n4 = Multiply(
   A=ScalarParameter() {defaultValue=0.2, group=Quest float down, parameterName=Rotate Y angle}
   B=LinearInterpolate(Alpha=NamedRerouteUsage()) {constA=1, constB=0}
  )
  n5 = Sine(None=Multiply(A=Time(), B=ScalarParameter() {defaultValue=0.12, group=Quest float down, parameterName=Speed Y Angle}) {constB=0.1})
  n6 = Add(
   A=MaterialFunctionCall() {materialFunction=&ObjectPivotPoint}.Object Pivot Location
   B=VectorParameter() {group=Quest float down, parameterName=CenterPositionOffset, defaultValue=<rgba 0 0 0 1>}.RGB
  )
  n7 = Add(
   A=MaterialFunctionCall(X=Constant(), Y=Constant(), Z=LinearInterpolate(A=@n1, Alpha=@n3) {constB=0}) {materialFunction=&MakeFloat3}.Result
   B=MaterialFunctionCall!(
    RotationAngle=Multiply(A=@n4, B=@n5)
    NormalizedRotationAxis=Constant3Vector() {constant=<rgba 0 1 0 0>}
    PivotPoint=@n6
    Position
    Normal
    TangentU
    TangentV
   ) {materialFunction=/Game/Characters/Player/Male/Medium/Bodies/M_MED_Inferno/FX/Mat/RotateVertexAboutAxis}.WPO
  )
  n8 = ScalarParameter() {defaultValue=0.1, group=Quest float down, parameterName=Rotate X angle}
  n9 = LinearInterpolate(Alpha=NamedRerouteUsage()) {constA=1, constB=0}
  n10 = Multiply(A=Time(), B=ScalarParameter() {defaultValue=0.12, group=Quest float down, parameterName=Speed X Angle}) {constB=0.1}
  n11 = MaterialFunctionCall(
   RotationAngle=Multiply(A=Multiply(A=@n8, B=@n9), B=Sine(None=@n10))
   NormalizedRotationAxis=Constant3Vector() {constant=<rgba 1 0 0 0>}
   PivotPoint=@n6
  ) {materialFunction=&RotateVertexAboutAxis}
  n12 = NamedRerouteUsage()
  n13 = Time()
  n14 = ScalarParameter() {defaultValue=0.1, group=Quest float down, parameterName=Speed Z Angle}
  n15 = Multiply(
   A=Multiply(
    A=ScalarParameter() {defaultValue=0.05, group=Quest float down, parameterName=Rotate Z angle}
    B=LinearInterpolate(Alpha=@n12) {constA=1, constB=0}
   )
   B=Sine(None=Multiply(A=@n13, B=@n14) {constB=0.1})
  )
  n16 = Constant3Vector() {constant=<rgba 0 0 1 0>}
  FunctionOutput(None=Add(
    A=Add(A=@n7, B=@n11.WPO)
    B=MaterialFunctionCall(RotationAngle=@n15, NormalizedRotationAxis=@n16, PivotPoint=@n6) {materialFunction=&RotateVertexAboutAxis}.WPO
   ))
  MaterialFunctionCall() {materialFunction=&ObjectPivotPoint}
  n17 = ScalarParameter() {bUseCustomPrimitiveData=true, defaultValue=1, group=Quest float down, parameterName=Float down animation, primitiveDataIndex=2}
  NamedRerouteDeclaration(None=@n17) {name=Timeianimation, nodeColor=<rgba 1 0 0.6562 1>}
  OneMinus(None=@n3)
  Power(Base=@n2) {constExponent=-2}
  Subtract(A=ActorPositionWS(), B=WorldPosition().XYZ)
  Subtract()
  Tangent!(None=@n2)
  Tangent(None=@n2)
 ]}
