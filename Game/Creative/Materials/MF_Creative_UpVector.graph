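# MF_Creative_UpVector is a Creative material function that overlays a diffuse color on upward-facing surfaces of a material.
# The saturated world-space Z of the vertex normal serves as the mask, blending BaseColor toward the UpVectorTexture input, which defaults to light grey.
# The same mask flattens Normal toward (0,0,1).
# Static switch UseUpVectorTexture outputs either the modified attributes or the untouched input.
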MaterialFunction /Game/Creative/Materials/MF_Creative_UpVector
graph {
 properties {description=This function uses world space to overlay a diffuse color on the z-axis.}
 nodes [
  n1 = FunctionInput() {inputType=FunctionInput_MaterialAttributes}
  n2 = GetMaterialAttributes(None=@n1) {attributeGetTypes [#1, #2]}
  n3 = FunctionInput() {bUsePreviewValueAsDefault=true, inputName=UpVectorTexture, sortPriority=1, previewValue=<0.7 0.7 0.7 1>}
  n4 = Saturate(None=ComponentMask(None=VertexNormalWS()) {b=true})
  n5 = LinearInterpolate(A=@n2.BaseColor, B=@n3, Alpha=@n4)
  n6 = LinearInterpolate(A=@n2.Normal, B=Constant3Vector() {constant=<rgba 0 0 1 0>}, Alpha=@n4)
  FunctionOutput(None=StaticSwitchParameter(
    True=SetMaterialAttributes!(MaterialAttributes=@n2.MaterialAttributes, BaseColor=@n5, Normal=@n6) {attributeSetTypes [#1, #2]}
    False=@n1
   ) {group=UpVector, parameterName=UseUpVectorTexture})
 ]
}
