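# Material function in the PrincessCastle environment set, named for Creative customization, that applies the castle sand treatment to incoming material attributes.
# It is a thin wrapper: a single Material Attributes input feeds MF_PrincessCastle_Sand, whose Result becomes the function output.
# It adds no other logic or parameters of its own.
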MaterialFunction /Game/Environments/Sets/PrincessCastle/MaterialFunctions/MF_Creative_Customization
graph {nodes [FunctionOutput(None=MaterialFunctionCall!(In=FunctionInput() {inputType=FunctionInput_MaterialAttributes}) {materialFunction=&/MF_PrincessCastle_Sand}.Result)]}
