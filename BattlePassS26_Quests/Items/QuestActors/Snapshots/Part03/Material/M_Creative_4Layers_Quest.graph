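# Four-layer Creative master material for Battle Pass S26 quest snapshot actors, adding MF_Quest_FloatDown to World Position Offset.
# Each layer's tinted diffuse, masks, normal and emissive, with optional high-quality POM UVs, feeds a texture-set lerp.
# Creative Colors multiplies the blended diffuse and emissive, then decal, fuzzy, time-of-day emissive, up-vector, customization and shared functions follow.
# Mobile simplifies layer 4, and ES2 keeps only base color before customization.
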Material /BattlePassS26_Quests/Items/QuestActors/Snapshots/Part03/Material/M_Creative_4Layers_Quest
graph {
 properties {
  bAutomaticallySetUsageInEditor=false
  bUseMaterialAttributes=true
  bUsedWithClothing=true
  bUsedWithInstancedStaticMeshes=true
  bUsedWithMeshDeformer=true
  bUsedWithNanite=true
  bUsedWithSkeletalMesh=true
  ditheredLODTransition=true
  pixelDepthOffsetMode=PDOM_Legacy
  usedShadingModels=Default Lit
 }
 nodes [
  n1 = MaterialFunctionCall() {materialFunction=/Game/Athena/WinterArt/Materials/MF_UVsSet}
  n2 = TextureObjectParameter() {group=Material_1, parameterName=Normals, samplerType=SAMPLERTYPE_Normal, sortPriority=2, texture=/Engine/EngineMaterials/FlatNormal}
  n3 = ScalarParameter() {defaultValue=1, group=Material_1, parameterName=Mat1 UV Scale, sortPriority=3}
  n4 = Multiply(A=TextureCoordinate(), B=Clamp(None=Ceil(None=@n3)) {maxDefault=1000, minDefault=1})
  n5 = TextureSample!(UVs=@n1.POM UVs 4 Layers, Tex=@n2, "DDX(UVs)"=DDX(Value=@n4), "DDY(UVs)"=DDY(Value=@n4), Apply View MipBias) {mipValueMode=TMVM_Derivative, samplerType=SAMPLERTYPE_Normal}
  n6 = TextureSample!(UVs=@n1.UVs Layer 1, Tex=@n2, Apply View MipBias) {samplerType=SAMPLERTYPE_Normal}
  n7 = QualitySwitch(
   Default=@n1.UVs Layer 1
   High=StaticSwitchParameter(True=@n1.POM UVs 4 Layers, False=@n1.POM UVs 4 Layers) {group=Material_1, parameterName=POM_True, sortPriority=11}
  )
  n8 = TextureObjectParameter() {
   group=Material_1
   parameterName=SpecularMasks
   samplerType=SAMPLERTYPE_Masks
   sortPriority=1
   texture=&T_Fortnite_Default_S
  }
  n9 = TextureSample(UVs=@n7, Tex=@n8) {samplerType=SAMPLERTYPE_Masks, texture=/Game/Environments/Sets/IceCastle/Textures/T_IceCastle_Wall1_D}
  n10 = MaterialFunctionCall() {materialFunction=&MF_UVsSet}
  n11 = QualitySwitch(
   Default=@n10.UVs Layer 3
   High=StaticSwitchParameter(True=@n10.POM UVs 4 Layers, False=@n10.POM UVs 4 Layers) {group=Material_3, parameterName=L3_POM_True, sortPriority=11}
  )
  n12 = TextureSampleParameter2D(UVs=@n11) {
   desc=Use for trims and accents
   group=Material_3
   parameterName=SpecularMasks_3
   samplerSource=SSM_Wrap_WorldGroupSettings
   samplerType=SAMPLERTYPE_Masks
   sortPriority=2
   texture=&T_Fortnite_Default_S
  }
  n13 = TextureObjectParameter() {group=Material_1, parameterName=Diffuse, texture=&T_IceCastle_Wall1_D}
  n14 = TextureSample(UVs=@n7, Tex=@n13) {texture=&T_IceCastle_Wall1_D}
  n15 = VectorParameter() {group=Material_1, parameterName=M1_TintColor, sortPriority=5, defaultValue=<rgba 1 1 1 0>}
  n16 = Multiply(A=@n14.R) {constB=6}
  n17 = Multiply(
   A=@n14.R
   B=LinearInterpolate(
    A=VectorParameter() {group=Material_1, parameterName=M1_TintColorB, sortPriority=5, defaultValue=<rgba 1 1 1 0>}.RGB
    B=@n15.RGB
    Alpha=Multiply(A=@n16, B=@n14.G)
   )
  )
  n18 = TextureSample(UVs=@n7, Tex=@n13) {texture=&T_IceCastle_Wall1_D}
  n19 = LinearInterpolate(
   A=Multiply(A=MaterialFunctionCall!(In=@n15.RGB) {materialFunction=/Game/Athena/WinterArt/Materials/MF_VPTintMask}.Result, B=@n18.RGB)
   B=@n18.RGB
   Alpha=@n18.RGB
  )
  n20 = TextureSampleParameter2D(UVs=@n10.UVs Layer 3) {
   desc=Use for trims and accents
   group=Material_3
   parameterName=Normals_Texture_3
   samplerSource=SSM_Wrap_WorldGroupSettings
   samplerType=SAMPLERTYPE_Normal
   sortPriority=3
   texture=&FlatNormal
  }
  n21 = TextureSampleParameter2D!(UVs=@n10.POM UVs 4 Layers, "DDX(UVs)"=DDX(Value=@n10.UVs Layer 3), "DDY(UVs)"=DDY(Value=@n10.UVs Layer 3), Apply View MipBias) {
   desc=Use for trims and accents
   group=Material_3
   mipValueMode=TMVM_Derivative
   parameterName=Normals_Texture_3
   samplerSource=SSM_Wrap_WorldGroupSettings
   samplerType=SAMPLERTYPE_Normal
   sortPriority=3
   texture=&FlatNormal
  }
  n22 = VectorParameter() {group=Material_4, parameterName=SpecularMasksks_4_MobileOverride, defaultValue=<rgba 0.5 0 0.5 0>}
  n23 = MaterialFunctionCall() {materialFunction=&MF_UVsSet}
  n24 = QualitySwitch(
   Default=@n23.UVs Layer 4
   High=StaticSwitchParameter(True=@n23.POM UVs 4 Layers, False=@n23.POM UVs 4 Layers) {group=Material_4, parameterName=L4_POM_True, sortPriority=11}
  )
  n25 = TextureSampleParameter2D!(UVs=@n24, Apply View MipBias) {
   desc=Use for doors
   group=Material_4
   parameterName=SpecularMasks_4
   samplerSource=SSM_Wrap_WorldGroupSettings
   samplerType=SAMPLERTYPE_Masks
   sortPriority=2
   texture=&T_Fortnite_Default_S
  }
  n26 = TextureSampleParameter2D(UVs=@n11) {
   desc=Use for trims and accents
   group=Material_3
   parameterName=Diffuse_Texture_3
   samplerSource=SSM_Wrap_WorldGroupSettings
   sortPriority=1
   texture=/Game/Global/Textures/Default/Blanks/T_MidgreysRGB
  }
  n27 = Multiply(
   A=MaterialFunctionCall(In=VectorParameter() {group=Material_3, parameterName=M3_TintColor, sortPriority=5, defaultValue=<rgba 1 1 1 0>}.RGB) {materialFunction=&MF_VPTintMask}.Result
   B=@n26.RGB
  )
  n28 = TextureSampleParameter2D(UVs=@n24) {
   desc=Use for doors
   group=Material_4
   parameterName=Diffuse_Texture_4
   samplerSource=SSM_Wrap_WorldGroupSettings
   sortPriority=1
   texture=&T_MidgreysRGB
  }
  n29 = Multiply(
   A=MaterialFunctionCall(In=VectorParameter() {group=Material_4, parameterName=M4_TintColor, sortPriority=5, defaultValue=<rgba 1 1 1 0>}.RGB) {materialFunction=&MF_VPTintMask}.Result
   B=@n28.RGB
  )
  n30 = TextureSampleParameter2D() {
   desc=Use for doors
   group=Material_4
   parameterName=Normals_Texture_4
   samplerSource=SSM_Wrap_WorldGroupSettings
   samplerType=SAMPLERTYPE_Normal
   sortPriority=3
   texture=&FlatNormal
  }
  n31 = TextureSampleParameter2D!(UVs=@n23.POM UVs 4 Layers, "DDX(UVs)"=DDX(Value=@n23.UVs Layer 4), "DDY(UVs)"=DDY(Value=@n23.UVs Layer 4), Apply View MipBias) {
   desc=Use for doors
   group=Material_4
   mipValueMode=TMVM_Derivative
   parameterName=Normals_Texture_4
   samplerSource=SSM_Wrap_WorldGroupSettings
   samplerType=SAMPLERTYPE_Normal
   sortPriority=3
   texture=&FlatNormal
  }
  n32 = StaticSwitchParameter(True=@n31.RGB, False=@n30.RGB) {group=Material_4, parameterName=L4_POM_True, sortPriority=11}
  n33 = MaterialFunctionCall() {materialFunction=&MF_UVsSet}
  n34 = QualitySwitch(
   Default=@n33.UVs Layer 2
   High=StaticSwitchParameter(True=@n33.POM UVs 4 Layers, False=@n33.POM UVs 4 Layers) {group=Material_2, parameterName=L2_POM_True, sortPriority=11}
  )
  n35 = TextureSampleParameter2D!(UVs=@n34, Apply View MipBias) {
   desc=Use for interior walls and ceilings
   group=Material_2
   parameterName=Diffuse_Texture_2
   samplerSource=SSM_Wrap_WorldGroupSettings
   sortPriority=1
   texture=&T_MidgreysRGB
  }
  n36 = Multiply(
   A=MaterialFunctionCall(In=VectorParameter() {group=Material_2, parameterName=M2_TintColor, sortPriority=5, defaultValue=<rgba 1 1 1 0>}.RGB) {materialFunction=&MF_VPTintMask}.Result
   B=@n35.RGB
  )
  n37 = TextureSampleParameter2D(UVs=@n34) {
   desc=Use for interior walls and ceilings
   group=Material_2
   parameterName=SpecularMasks_2
   samplerSource=SSM_Wrap_WorldGroupSettings
   samplerType=SAMPLERTYPE_Masks
   sortPriority=2
   texture=&T_Fortnite_Default_S
  }
  n38 = TextureSampleParameter2D(UVs=@n33.UVs Layer 2) {
   desc=Use for interior walls and ceilings
   group=Material_2
   parameterName=Normals_Texture_2
   samplerSource=SSM_Wrap_WorldGroupSettings
   samplerType=SAMPLERTYPE_Normal
   sortPriority=3
   texture=&FlatNormal
  }
  n39 = TextureSampleParameter2D!(UVs=@n33.POM UVs 4 Layers, "DDX(UVs)"=DDX(Value=@n33.UVs Layer 2), "DDY(UVs)"=DDY(Value=@n33.UVs Layer 2), Apply View MipBias) {
   desc=Use for interior walls and ceilings
   group=Material_2
   mipValueMode=TMVM_Derivative
   parameterName=Normals_Texture_2
   samplerSource=SSM_Wrap_WorldGroupSettings
   samplerType=SAMPLERTYPE_Normal
   sortPriority=3
   texture=&FlatNormal
  }
  n40 = MaterialFunctionCall!(
   Use3Materials
   Use4Materials
   Use2Materials
   D=StaticSwitchParameter(True=AppendVector(A=@n17, B=@n14.B), False=AppendVector(A=@n19, B=@n18.A)) {group=Material_1, parameterName=CreativeColors}
   S=AppendVector(A=@n9.RGB, B=@n9.RGB)
   N=QualitySwitch(Default=@n6.RGB, High=StaticSwitchParameter(True=@n5.RGB, False=@n6.RGB) {group=Material_1, parameterName=POM_True, sortPriority=11})
   D2=AppendVector(A=LinearInterpolate(A=@n36, B=@n35.RGB, Alpha=@n35.RGB), B=@n35.A)
   S2=AppendVector(A=@n37.RGB, B=@n37.RGB)
   N2=QualitySwitch(Default=@n38.RGB, High=StaticSwitchParameter(True=@n39.RGB, False=@n38.RGB) {group=Material_2, parameterName=L2_POM_True, sortPriority=11})
   D3=AppendVector(A=LinearInterpolate(A=@n27, B=@n26.RGB, Alpha=@n26.RGB), B=@n26.A)
   S3=AppendVector(A=@n12.RGB, B=@n12.RGB)
   N3=QualitySwitch(Default=@n20.RGB, High=StaticSwitchParameter(True=@n21.RGB, False=@n20.RGB) {group=Material_3, parameterName=L3_POM_True, sortPriority=11})
   D4=AppendVector(A=LinearInterpolate(A=@n29, B=@n28.RGB, Alpha=@n28.RGB), B=@n28.A)
   S4=ShadingPathSwitch(Default=AppendVector(A=@n25.RGB, B=@n25.RGB), Mobile=AppendVector(A=@n22.RGB, B=@n22.RGB))
   N4=ShadingPathSwitch(Default=QualitySwitch(Default=@n30.RGB, High=@n32), Mobile=VertexNormalWS())
   E1=StaticSwitchParameter(
    True=VectorParameter() {bUseCustomPrimitiveData=true, parameterName=Emissive1CustomPrimitive, primitiveDataIndex=28}.RGB
    False=VectorParameter() {group=Material_1, parameterName=Emissive, sortPriority=7}.RGB
   ) {parameterName=Emissive1UsesCustomPrimitive}
   E2=VectorParameter() {group=Material_2, parameterName=Emissive2, sortPriority=7}.RGB
   E3=VectorParameter() {group=Material_3, parameterName=Emissive3, sortPriority=7}.RGB
   E4=VectorParameter() {group=Material_4, parameterName=Emissive4, sortPriority=7}.RGB
   CCMaskByMetal_1
   CC1
   CCR_1
   CC2
   CCR_2
   CCMaskByMetal_2
   CC3
   CCR_3
   CCMaskByMetal_3
   CC4
   CCMaskByMetal_4
   CCR_4
   OpacityMaskCancel_1
   OpacityMaskCancel_2
   OpacityMaskCancel_3
   OpacityMaskCancel_4
   SSS_1
   SSS_2
   SSS_3
   SSS_4
   SSS_Depth_1
   SSS_Depth_2
   SSS_Depth_3
   SSS_Depth_4
  ) {materialFunction=/Game/Athena/WinterArt/Materials/MF_FN_TextureSetLerp_4Layers}
  n41 = MaterialFunctionCall!(Mask=@n40.Height, Index) {materialFunction=/Game/Creative/Materials/MF_CreativeColors}
  n42 = Multiply(A=@n41.Result, B=@n40.Diffuse)
  n43 = SetMaterialAttributes!(
   MaterialAttributes
   Base Color=@n42
   Specular=@n40.S
   Metallic=@n40.S
   Roughness=@n40.S
   Emissive Color=Multiply(A=@n41.Result, B=@n40.Emissive)
   Normal=@n40.S
   Opacity=@n40.S
  ) {attributeSetTypes [
    #1
    #2
    #3
    #4
    #5
    #6
    #7
   ]}
  n44 = MaterialFunctionCall!(In=@n43) {materialFunction=/Game/Athena/WinterArt/Materials/MF_DecalPrint}
  n45 = MaterialFunctionCall!(
   In=MaterialFunctionCall!(In=MaterialFunctionCall!(In=@n44.Result) {materialFunction=/Game/Athena/WinterArt/Materials/MF_Fuzzy}.Result) {materialFunction=/Game/Athena/WinterArt/Materials/MF_TOD_EmissiveColors}.Result
   UpVectorTexture=VectorParameter() {group=Global, parameterName=UpVectorColor, defaultValue=<rgba 0.8125 0.8125 0.8125 0>}.RGB
  ) {materialFunction=/Game/Creative/Materials/MF_Creative_UpVector}
  n46 = FeatureLevelSwitch(
   Default=@n45.Result
   ES2=SetMaterialAttributes!(MaterialAttributes, Base Color=@n42) {attributeSetTypes [#1]}
   ES3_1=@n45.Result
   SM4_REMOVED=@n45.Result
   SM5=@n45.Result
   SM6=@n45.Result
  )
  n47 = GetMaterialAttributes(None=MaterialFunctionCall!(In=MaterialFunctionCall!(In=@n46) {materialFunction=/Game/Environments/Sets/PrincessCastle/MaterialFunctions/MF_Creative_Customization}.Result) {materialFunction=/Game/Athena/WinterArt/Materials/MF_GameWideShared}.Result) {attributeGetTypes [#8]}
  SetMaterialAttributes!(
   MaterialAttributes=@n47.MaterialAttributes
   World Position Offset=Add(
    A=@n47.World Position Offset
    B=MaterialFunctionCall() {materialFunction=&/MF_Quest_FloatDown}.Result
   )
  ) {attributeSetTypes [#8]}
 ]
}
